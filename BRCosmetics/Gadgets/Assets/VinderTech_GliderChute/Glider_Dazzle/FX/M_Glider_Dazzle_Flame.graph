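# Unlit translucent flame material for the Dazzle glider's effects, built for Niagara mesh particles.
# A scrolling normal texture distorts panned noise, which is multiplied by sine-and-gradient UV masks and sharpened by ConfigShape gain and exponent into a flame mask.
# Particle color feeds an emissive adjustment function, while the mask gets particle alpha, view-angle and depth fades.
# Depth fade is bypassed on mobile through a shading path switch.
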Material /BRCosmetics/Gadgets/Assets/VinderTech_GliderChute/Glider_Dazzle/FX/M_Glider_Dazzle_Flame
graph {
 properties {
  bUsedWithNiagaraMeshParticles=true
  blendMode=BLEND_Translucent
  pixelDepthOffsetMode=PDOM_Legacy
  shadingModel=MSM_Unlit
  usedShadingModels=Unlit
 }
 nodes [
  n1 = ParticleRandom()
  AppendVector(A=@n1, B=@n1)
  n2 = ParticleRandom()
  AppendVector(A=@n2, B=@n2)
  Constant4Vector() {constant=<rgba 1.3 0.8 -0.1 0.4>}
  n3 = TextureCoordinate()
  n4 = Constant4Vector() {constant=<rgba 1.3 0.7 0.1 0.5>}
  n5 = Panner(Coordinate=Multiply(A=@n3, B=AppendVector(A=@n4.R, B=@n4.R)), Speed=AppendVector(A=@n4.B, B=@n4.B))
  n6 = TextureCoordinate()
  n7 = Constant4Vector() {constant=<rgba 1 2 0 1>}
  n8 = Multiply(A=@n6, B=AppendVector(A=@n7.R, B=@n7.R))
  n9 = AppendVector(A=@n7.B, B=@n7.B)
  n10 = TextureSampleParameter2D(UVs=Add(B=Panner(Coordinate=@n8, Speed=@n9))) {parameterName=Tex1, samplerType=SAMPLERTYPE_Normal, texture=&T_Weap_Distort_N_02}
  n11 = OneMinus(None=ComponentMask(None=@n6) {g=true})
  n12 = ScalarParameter() {defaultValue=1, group=Distortion, parameterName=Distortion Gradient Lerp, sortPriority=3}
  n13 = Multiply(
   A=Multiply(A=ComponentMask(None=@n10.RGB) {g=true, r=true}, B=LinearInterpolate(B=@n11, Alpha=@n12))
   B=ScalarParameter() {defaultValue=0.1, group=Distortion, parameterName=Distortion Intensity Noise, sortPriority=2}
  )
  n14 = TextureSampleParameter2D(UVs=Add(A=Add(B=@n5), B=@n13)) {group=Noise, parameterName=NoIse A Texture, sortPriority=0, texture=&T_ShieldCaustics}
  n15 = TextureCoordinate()
  n16 = VectorParameter() {parameterName=TexOptions, defaultValue=<rgba 1.3 0.8 -0.1 0.4>}
  n17 = Multiply(A=@n15, B=AppendVector(A=@n16.R, B=@n16.R))
  n18 = AppendVector(A=@n16.B, B=@n16.B)
  n19 = ScalarParameter() {defaultValue=1, parameterName=Use_Distort}
  n20 = TextureSampleParameter2D(UVs=Add(A=Panner(Coordinate=@n17, Speed=@n18), B=LinearInterpolate(A=@n15, B=@n13, Alpha=@n19))) {group=Noise, parameterName=NoIse B Texture, sortPriority=4, texture=&T_NoisePack_01}
  LinearInterpolate(A=@n14.R, B=@n20.R)
  n21 = ComponentMask(None=@n3) {g=true}
  n22 = MaterialFunctionCall(Float2=@n3) {materialFunction=&BreakOutFloat2Components}
  n23 = Panner(Coordinate=AppendVector(A=Multiply(A=@n22.R) {constB=1.4}, B=Multiply(A=@n22.G) {constB=0.7})) {speedY=0.5}
  n24 = TextureCoordinate()
  n25 = ComponentMask(None=@n24) {r=true}
  n26 = ComponentMask(None=@n24) {g=true}
  n27 = MaterialFunctionCall(UVs=AppendVector(A=Add(A=@n25) {constB=0}, B=Add(A=@n26) {constB=-0.4}), Radius=Constant() {r=0.6}, Density=Constant() {r=2}) {materialFunction=&RadialGradientExponential}
  n28 = Multiply(
   A=Add(
    A=TextureSampleParameter2D(UVs=@n23) {group=Noise, parameterName=NoIse B Texture, sortPriority=4, texture=&T_NoisePack_01}.R
    B=Multiply(A=@n27.RadialGradientExponential)
   )
   B=@n21
  )
  n29 = Multiply(A=Power(Base=LinearInterpolate(A=@n28, B=@n14.R) {constAlpha=0.2}) {constExponent=1.8}) {constB=6}
  n30 = ParticleColor()
  LinearInterpolate(A=Saturate(None=Multiply(A=@n21, B=@n29)), Alpha=@n30.A)
  n31 = TextureCoordinate()
  n32 = Multiply(A=ComponentMask(None=@n31) {r=true}) {constB=0.5}
  n33 = Multiply(A=Power(Base=Sine(None=@n32), Exp=Constant() {r=1.3}) {constExponent=1}, B=Power(Base=ComponentMask(None=@n31) {g=true}))
  n34 = TextureCoordinate()
  n35 = ComponentMask(None=@n34) {g=true}
  n36 = ComponentMask(None=@n34) {r=true}
  n37 = Multiply(A=@n35, B=Sine(None=Multiply(A=@n36) {constB=0.5}))
  n38 = Multiply(A=LinearInterpolate(A=Multiply(A=@n20.R, B=@n33), B=@n37, Alpha=Multiply(A=@n37) {constB=0.03}) {constAlpha=0.01}, B=@n37)
  n39 = VectorParameter() {parameterName=ConfigShape, defaultValue=<rgba 200 1 0 20>}
  n40 = OneMinus(None=@n35)
  n41 = LinearInterpolate(A=@n39.R, B=@n39.R, Alpha=@n40) {constA=200}
  n42 = Saturate(None=Power(Base=Multiply(A=@n38, B=@n41) {constB=50}, Exp=LinearInterpolate(A=@n39.B, B=@n39.B, Alpha=@n40) {constB=20}) {constExponent=50})
  Multiply(
   A=@n42
   B=MaterialFunctionCall("FadeSensitivity (optional)"=ScalarParameter() {parameterName=Fade edge}) {materialFunction=&FadeBasedOnViewAngle}.Result
  )
  Multiply(A=MaterialFunctionCall(Emissive=@n30.RGB) {materialFunction=&MF_FX_EmissiveAdjustments}.Result) {constB=0.1}
  Multiply(A=@n30.A, B=@n42)
  n43 = DepthFade(Opacity=@n42, FadeDistance=ScalarParameter() {defaultValue=30, parameterName=Depth Fade Distance})
  ShadingPathSwitch(Default=@n43, Deferred=@n43, Mobile=@n42) {desc=Migrated from FeatureLevelSwitch}
 ]
}
