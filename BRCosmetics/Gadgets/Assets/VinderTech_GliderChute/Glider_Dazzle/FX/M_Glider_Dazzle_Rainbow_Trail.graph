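# Unlit, two-sided translucent material for the Dazzle glider's rainbow trail, set up for Niagara ribbons and mesh particles.
# Chained lerps blend five colors along V, optionally warped by panning distortion that fades along U, then square and tint them by particle color for emissive.
# Opacity combines length and edge fades, a sine tail wave driven by dynamic parameters, and radial gradients.
# A shading-path switch gives desktop depth fade and mobile an optional front fade.
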Material /BRCosmetics/Gadgets/Assets/VinderTech_GliderChute/Glider_Dazzle/FX/M_Glider_Dazzle_Rainbow_Trail
graph {
 properties {
  bUsedWithNiagaraMeshParticles=true
  bUsedWithNiagaraRibbons=true
  blendMode=BLEND_Translucent
  pixelDepthOffsetMode=PDOM_Legacy
  shadingModel=MSM_Unlit
  twoSided=true
  usedShadingModels=Unlit
 }
 nodes [
  n1 = ParticleRandom()
  AppendVector(A=@n1, B=@n1)
  ComponentMask(None=TextureCoordinate()) {r=true}
  Constant()
  n2 = ScalarParameter() {defaultValue=0.1, parameterName=FresnelExp}
  MaterialFunctionCall(Invert Fresnel=StaticBool() {value=true}, Power=@n2) {materialFunction=&Fresnel_Function}
  n3 = TextureCoordinate()
  n4 = Constant4Vector() {constant=<rgba 1 1 -1 -0.1>}
  n5 = Multiply(A=@n3, B=AppendVector(A=@n4.R, B=@n4.R))
  n6 = AppendVector(A=@n4.B, B=@n4.B)
  n7 = TextureSampleParameter2D(UVs=Add(B=Panner(Coordinate=@n5, Speed=@n6))) {parameterName=Distort_Tex, samplerType=SAMPLERTYPE_Normal, texture=&T_Weap_Distort_N_02}
  n8 = ComponentMask(None=@n7.RGB) {g=true, r=true}
  n9 = Multiply(A=@n8, B=ScalarParameter() {defaultValue=2.5, group=Distortion, parameterName=Distortion Intensity Noise, sortPriority=2})
  n10 = TextureCoordinate()
  n11 = TextureCoordinate()
  n12 = ComponentMask(None=@n11) {r=true}
  n13 = Multiply(A=@n12) {constB=2.8}
  n14 = StaticSwitchParameter(True=LinearInterpolate(A=Add(A=@n9, B=@n10), B=@n10, Alpha=Saturate(None=@n13)), False=TextureCoordinate()) {defaultValue=true, parameterName=Use_Distortion}
  n15 = Multiply(A=OneMinus(None=ComponentMask(None=@n14) {g=true})) {constB=4}
  n16 = LinearInterpolate(
   A=VectorParameter() {parameterName=Color 1, defaultValue=<rgba 1 0.085 0.2344 0.2>}.RGB
   B=VectorParameter() {parameterName=Color 2, defaultValue=<rgba 0.4531 0.3874 0 0.3>}.RGB
   Alpha=Saturate(None=@n15)
  )
  n17 = VectorParameter() {parameterName=Color 3, defaultValue=<rgba 0.0071 0.651 0.1788 0.2>}
  n18 = Subtract(A=@n15)
  n19 = Saturate(None=@n18)
  n20 = Subtract(A=@n18)
  n21 = LinearInterpolate(
   A=LinearInterpolate(
    A=LinearInterpolate(A=@n16, B=@n17.RGB, Alpha=@n19)
    B=VectorParameter() {parameterName=Color 4, defaultValue=<rgba 0.0098 0.4015 0.8906 0.1>}.RGB
    Alpha=Saturate(None=@n20)
   )
   B=VectorParameter() {parameterName=Color 5, defaultValue=<rgba 0.488 0.0703 0.7812 0.2>}.RGB
   Alpha=Saturate(None=Subtract(A=@n20))
  )
  MaterialFunctionCall(Emissive=Multiply(A=Multiply(A=@n21, B=@n21), B=ParticleColor().RGB)) {materialFunction=&MF_FX_EmissiveAdjustments}
  Multiply()
  Multiply(A=@n8, B=OneMinus(None=ComponentMask(None=@n3) {r=true}))
  OneMinus(None=Fresnel(ExponentIn=@n2))
  Power(Base=@n12) {constExponent=6}
  ScalarParameter() {defaultValue=0.6, parameterName=Length Fade}
  ScalarParameter() {defaultValue=0.7, parameterName=Tail Wave Speed}
  ScalarParameter() {defaultValue=0.1, parameterName=Tail Wave Intensity}
  n22 = OneMinus(None=@n12)
  n23 = ComponentMask(None=@n11) {g=true}
  n24 = Abs(None=ConstantBiasScale(None=@n23) {bias=-0.5, scale=2})
  n25 = Saturate(None=Divide(A=OneMinus(None=@n24), B=ScalarParameter() {defaultValue=0.2, parameterName=Edge Fade}))
  n26 = Multiply(A=Power(Base=@n22), B=@n25)
  n27 = DynamicParameter() {defaultValue=<rgba 0.7 0.1 0.1 0.6>, paramNames [TrailWaveSpeed, TrailWaveIntensity, EdgeFade, LengthFade]}
  n28 = Multiply(A=Add(A=@n23, B=Time()), B=@n27.TrailWaveSpeed)
  n29 = Multiply(A=ConstantBiasScale(None=Sine(None=@n28)), B=@n27.TrailWaveIntensity)
  n30 = Power(Base=Multiply(A=@n22, B=Subtract(A=@n22, B=@n29))) {constExponent=1.5}
  n31 = DynamicParameter() {defaultValue=<rgba 0.7 0.1 0.1 0.6>, paramNames [TrailWaveSpeed, TrailWaveIntensity, EdgeFade, LengthFade]}
  n32 = Multiply(A=Saturate(None=Divide(A=@n30, B=@n31.LengthFade)), B=@n25)
  n33 = TextureCoordinate()
  n34 = ComponentMask(None=@n33) {r=true}
  n35 = MaterialFunctionCall(UVs=AppendVector(A=Add(A=@n34) {constB=-0.2}, B=ComponentMask(None=@n33) {g=true}), Density=Constant() {r=4}) {materialFunction=&RadialGradientExponential}
  n36 = Multiply(A=@n32, B=OneMinus(None=Multiply(A=@n35.RadialGradientExponential)))
  n37 = Multiply(A=ParticleColor().A, B=Saturate(None=LinearInterpolate(A=@n26, B=@n36) {constAlpha=0.2}))
  n38 = TextureCoordinate()
  n39 = AppendVector(A=Add(A=ComponentMask(None=@n38) {r=true}) {constB=0.2}, B=ComponentMask(None=@n38) {g=true})
  n40 = Constant() {r=5}
  n41 = Multiply(
   A=@n13
   B=Add(
    A=@n13
    B=MaterialFunctionCall(UVs=@n39, Density=@n40) {materialFunction=&RadialGradientExponential}.RadialGradientExponential
   )
  )
  n42 = DepthFade(
   Opacity=LinearInterpolate(A=@n37, B=Multiply(A=@n41, B=@n37), Alpha=ScalarParameter() {parameterName=UseSoftEdge})
   FadeDistance=ScalarParameter() {defaultValue=300, parameterName=Depth Fade Distance}
  )
  n43 = Power(Base=ComponentMask(None=TextureCoordinate()) {r=true}, Exp=ScalarParameter() {defaultValue=2, parameterName=Mobile_Front_Fade_Amount}) {constExponent=1.3}
  n44 = Multiply(A=@n43) {constB=2}
  n45 = Saturate(None=LinearInterpolate(A=@n43, Alpha=Saturate(None=@n44)) {constB=5})
  ShadingPathSwitch(
   Default=@n42
   Deferred=@n42
   Mobile=LinearInterpolate(A=@n37, B=Multiply(A=@n37, B=@n45), Alpha=ScalarParameter() {parameterName=Mobile_Use_Front_Fade})
  ) {desc=Migrated from FeatureLevelSwitch}
  n46 = Divide(A=@n12, B=ScalarParameter() {defaultValue=0.3, parameterName=Base Fade})
  VertexInterpolator(VS=Multiply(A=Saturate(None=@n46), B=@n32))
 ]
}
