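# Niagara trail effect for the Dazzle glider, combining rainbow mesh trails, a goopy mesh trail, toon-shaded tricolor stars and sprite speed lines.
# Both trail types follow skeletal mesh bones or sockets, with a shared system timeline fading rainbows in and goop out.
# Smoothed ForwardDot, gated by deployment, drives star spawning.
# Speed lines fade with ForwardDot in gameplay or deployment in front-end preview, which also adjusts their density and placement.
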NiagaraSystem /BRCosmetics/Gadgets/Assets/VinderTech_GliderChute/Glider_Dazzle/FX/NS_Glider_Dazzle_Trails
dependencies {
 usedDataInterfaces [
  &NiagaraDataInterfaceArrayFloat
  &NiagaraDataInterfaceCurve
  &NiagaraDataInterfaceArrayColor
  &NiagaraDataInterfaceVector2DCurve
  &NiagaraDataInterfaceColorCurve
 ]
 usedDynamicInputs [
  &Boolean_OrOperation
  &Boolean_AndOperation
  &SetBoolByFloatComparison
 ]
 usedModules [&SystemState, &FN_Integrator, &Timeline]
}
emitters {
 Rainbow_Trails {
  emitterUpdateScript [
   &EmitterState()
   &SpawnBurst_Instantaneous(Spawn Count=4, Use Spawn Probability=false, UseLoopCountLimit=true)
   &InterpolateOverTime(Rate Of Change=4, Target Float Value:NiagaraFloat=$User.ForwardDot) {enabled=false}
  ]
  particleSpawnScript [
   &FN_SetSpawnGroup()
   &InitializeParticle(
    Color=&MakeLinearColorFromVectorAndFloat(
     "Vector (RGB)"=&MakeVectorFromLinearColorRGB(Color=<rgba 1.5 1.5 1.5 1>)
     "Float (Alpha)"=&SelectFloatFromArray(
      Direct Array Index=$Particles.SpawnGroup
      Direct Array Mode=::NewEnumerator1
      Float Selection Array=DataInterface() {propertyValues {FloatData [1,1,1,2]}}
     )
    )
    Mesh Scale=&Multiply_VectorByFloat(Vector=<2 1 0.96>, Float=0.4)
    Mesh Scale Mode=::NewEnumerator4
    Sprite Size Mode=::NewEnumerator3
    Uniform Sprite Size=1
    UsePositionOffset=false
    Write Material Random=true
   )
   &InitialMeshOrientation(
    Mesh Axis To Orient=<0 0 1>
    Mesh Orientation Mode=::NewEnumerator2
    Orientation Vector=<0 0 1>
    Rotation=<0 0 0.25>
    Use Rotation Vector=true
   )
   &DynamicMaterialParameters(
    Index 0 Param 1=&RandomRangeFloat(Minimum=0.5, Maximum=1.6)
    Module.Param0WriteEnabled=true
    Module.Param1WriteEnabled=false
    Module.Param2WriteEnabled=false
    Module.Param3WriteEnabled=false
   ) {module=DynamicMaterialParameters001}
  ]
  particleUpdateScript [
   &ParticleState(Kill Particles When Lifetime Has Elapsed=false, Let Infinitely Lived Particles Die When Emitter Deactivates=true)
   &SkeletalMeshLocation(
    Bone / Socket Index=&SpawnGroup()
    Bone Sampling Mode=::NewEnumerator4
    MaskSpawnGroup=false
    Orientation Sampling=::NewEnumerator1
    Skeletal Mesh=$User.SkeletalMesh
   )
   &UpdateMeshOrientation(
    Facing Direction=&RotateVectorByQuaternion(Vector To Rotate=<0 1 0>, Quaternion=$Particles.SkeletalMeshLocation.SampledRotation)
    Facing Interpolation=false
    Orientation Method=::NewEnumerator1
    Override Quat=false
    TransformFacing=false
    TransformUp=false
    Up Direction:Vector3f=<1 0 0>
   )
   &ScaleMeshSize(Scale Factor=&Multiply_VectorByFloat(
     Vector=&MakeVector(
      X=&FloatFromCurve(
       FloatCurve=DataInterface() {propertyValues {Curve {keys [<at 0 0 RCIM_Cubic>,<at 0.7 1 RCIM_Cubic>]}}}
       CurveIndex=$System.Timeline001.NormalizedTime
       Scale Curve=1.3
      )
      Y=1
      Z=1
     )
     Float=&SelectFloatFromArray(
      Direct Array Index=$Particles.SpawnGroup
      Direct Array Mode=::NewEnumerator1
      Float Selection Array=DataInterface() {propertyValues {FloatData [1.1,1.5,1.3,3.2]}}
     )
    ))
   &DynamicMaterialParameters(
    Index 0 Param 1=1
    Index 0 Param 2=&Lerp_Float(A=0.1, B=0.4, Alpha=$System.Timeline001.NormalizedTime)
    Index 0 Param 3=&Lerp_Float(A=0.8, B=0.3, Alpha=$System.Timeline001.NormalizedTime)
    Index 0 Param 4=&Lerp_Float(A=0.3, B=0.1, Alpha=$System.Timeline001.NormalizedTime)
    Module.Param0WriteEnabled=false
    Module.Param1WriteEnabled=true
    Module.Param2WriteEnabled=true
    Module.Param3WriteEnabled=true
   )
   &ScaleColor(
    Scale Alpha=&FloatFromCurve(
     FloatCurve=DataInterface() {propertyValues {Curve {keys [<at 0 0 RCIM_Cubic>,<at 1 1 RCIM_Cubic>]}}}
     CurveIndex=$System.Timeline001.NormalizedTime
     Scale Curve=&FN_Select_Float_from_Time_Of_Day(Morning=0.8, Day=1, Evening=0.8, Night=0.4)
    )
    ScaleA=true
    ScaleRGB=true
   )
  ]
  rendererClasses [&NiagaraMeshRendererProperties]
  renderers [
   {rendererClass=&NiagaraMeshRendererProperties, rendererIndex=0}
   {rendererClass=&NiagaraMeshRendererProperties, rendererIndex=1}
   {rendererClass=&NiagaraMeshRendererProperties, rendererIndex=2}
   {rendererClass=&NiagaraMeshRendererProperties, rendererIndex=3}
  ]
 }
 SpeedLines {
  emitterUpdateScript [
   &EmitterState()
   &SpawnRate(
    SpawnRate=&MakeCustomFloatFromBool(Bool=$User.bIsFrontEndPreview, True Float=50, False Float=90)
    Use Spawn Probability=true
   ) {module=SpawnRate001}
  ]
  particleSpawnScript [
   &InitializeParticle(
    Color=&MakeLinearColorFromVectorAndFloat(
     "Vector (RGB)"=<1 1 1>
     "Float (Alpha)"=&Multiply_Float(
      A=&UniformRangedFloat(Minimum=0.15, Maximum=0.3)
      B=&MakeCustomFloatFromBool(Bool=$User.bIsFrontEndPreview, True Float=0.25, False Float=0.8)
     )
    )
    Lifetime Max=0.16
    Lifetime Min=0.1
    Lifetime Mode=::NewEnumerator1
    Sprite Size Max=<3 10>
    Sprite Size Min=<2 2>
    Sprite Size Mode=::NewEnumerator2
    UsePositionOffset=false
    Write Material Random=true
   )
   &TorusLocation(
    Align Torus With Axis=true
    Handle Radius=0
    Large Radius:NiagaraFloat=&FN_FrontEndFloatModifier(
     Gameplay=&RandomRangeFloat(Minimum=190, Maximum=210)
     Front End=&RandomRangeFloat(Minimum=100, Maximum=130)
    )
    MaskSpawnGroup=false
    Module.Override Randomness=false
    Offset=<0 -180 0>
    Torus Axis=<0 1 0>
    Torus Origin:NiagaraPosition=&Add_Vector(
     A=Linked() {linkedVariable {name=Particles.Position, type=NiagaraPosition}}
     B=&MakeVector(
      X=0
      Y=&MakeCustomFloatFromBool(Bool=$User.bIsFrontEndPreview, True Float=80, False Float=-100)
      Z=&MakeCustomFloatFromBool(Bool=$User.bIsFrontEndPreview, True Float=0, False Float=20)
     )
    )
   )
   &AddVelocityFromPoint(
    Module.Use Falloff Distance=false
    Origin Offset=<0 -700 0>
    Velocity Strength=&Multiply_Float(A=500, B=&MakeCustomFloatFromBool(Bool=$User.bIsFrontEndPreview, True Float=0, False Float=1))
   )
   &AddVelocity(Velocity=&Multiply_Vector(
     A=&RandomRangeVector(Minimum=<0 300 0>, Maximum=<0 700 0>)
     B=&MakeVector(X=1, Y=&MakeCustomFloatFromBool(Bool=$User.bIsFrontEndPreview, True Float=1, False Float=1), Z=1)
    ))
  ]
  particleUpdateScript [
   &SolveForcesAndVelocity(Clamp Velocity=false, Limit Acceleration=false)
   &ParticleState()
   &ScaleSpriteSize(Scale Factor=&Vector2DFromCurve(
     Vector2Curve=DataInterface() {propertyValues {XCurve {keys [<at 0 0.5 RCIM_Cubic>,<at 0.5 1 RCIM_Cubic>]}, YCurve {keys [<at 0 0 RCIM_Cubic>,<at 0.5 1 RCIM_Cubic>]}}}
     ScaleCurve=&Vector2DFromFloat(Value=1)
    ))
   &ScaleSpriteSizeBySpeed(Max Scale Factor=<1 50>, Min Scale Factor=<0.01 0.05>, Velocity Threshold=1)
   &ScaleColor(
    Scale Alpha=&FloatFromCurve(
     FloatCurve=DataInterface() {propertyValues {Curve {keys [
         <at 0 0 RCIM_Cubic>
         <at 0.3759 1>
         {arriveTangent=-2.221, interpMode=RCIM_Cubic, leaveTangent=-2.221, tangentMode=RCTM_User, time=1, value=0}
        ]}}}
     Scale Curve=&MakeCustomFloatFromBool(
      Bool=$User.bIsFrontEndPreview
      True Float=&MakeCustomFloatFromBool(Bool=$User.bIsFullyDeployed, True Float=1, False Float=0)
      False Float=&Lerp_Float(A=0, B=1, Alpha=$User.ForwardDot)
     )
    )
    Scale RGB=&VectorFromFloat(Value=2)
    ScaleA=true
    ScaleRGB=true
   )
  ]
  rendererClasses [&NiagaraSpriteRendererProperties]
  renderers [{rendererClass=&NiagaraSpriteRendererProperties}]
 }
 StarToon {
  emitterSpawnScript [
   /Game/Effects/Niagara/MasterMaterials/ToonShader/Template/NM_ToonShaderParameters(
    Lighting_Hardness=1
    Metallic_LightingPower=0
    Metallic_Reflection_Color:LinearColor=<rgba 0 1 0.4868 1>
    RimLight_Color:LinearColor=<rgba 0.0377 2 0 1>
    RimLight_Width=0
   )
   &NM_ToonShaderParameters(
    Lighting_Hardness=1
    Metallic_LightingPower=0
    Metallic_Reflection_Color:LinearColor=<rgba 1 0.3125 0.8693 1>
    RimLight_Color:LinearColor=<rgba 2 0 1.8443 1>
    RimLight_Width=0
   ) {module=NM_ToonShaderParameters002}
   &NM_ToonShaderParameters(
    Lighting_Hardness=1
    Metallic_LightingPower=0
    Metallic_Reflection_Color:LinearColor=<rgba 1 0.7823 0 1>
    RimLight_Color:LinearColor=<rgba 2 1.1268 0 1>
    RimLight_Width=0
   ) {module=NM_ToonShaderParameters001}
   /Game/Effects/Niagara/MasterMaterials/ToonShader/Template/NM_OutlineParameters(OutlineColor:LinearColor=<rgba 0 0.6458 0.0139 1>)
   &NM_OutlineParameters(OutlineColor:LinearColor=<rgba 0.6198 0 0.348 1>) {module=NM_OutlineParameters002}
   &NM_OutlineParameters(OutlineColor:LinearColor=<rgba 0 0.3465 1 1>, OutlineWidth=3) {module=NM_OutlineParameters001}
  ]
  emitterUpdateScript [
   &EmitterState()
   &InterpolateOverTime(
    Rate Of Change=6
    Target Float Value:NiagaraFloat=&MakeCustomFloatFromBool(Bool=$User.bIsFullyDeployed, True Float=$User.ForwardDot, False Float=0)
   )
   &SpawnRate(
    Spawn Probability=&RandomRangeFloat(Minimum=0.5, Maximum=0.7)
    SpawnRate=&Multiply_Float(
     A=20
     B=&FloatFromCurve(
      FloatCurve=DataInterface() {propertyValues {Curve {keys [<at 0 0>,<at 1 1>]}}}
      CurveIndex=$Emitter.InterpolateOverTime.InterpolateOverTime_Float.Moving Average
     )
    )
    Use Spawn Probability=true
   )
  ]
  particleSpawnScript [
   &FN_SetSpawnGroup()
   &InitializeParticle(
    Color Mode=::NewEnumerator0
    Lifetime Max=0.5
    Lifetime Min=0.3
    Lifetime Mode=::NewEnumerator1
    Mesh Scale=&Multiply_VectorByFloat(Vector=<1 1 1>, Float=&RandomRangeFloat(Minimum=0.8, Maximum=1.3))
    Mesh Scale Mode=::NewEnumerator4
    UsePositionOffset=false
    Write Material Random=true
   )
   &SkeletalMeshLocation(
    Bone / Socket Index=3
    Bone Sampling Mode=::NewEnumerator4
    MaskSpawnGroup=false
    Skeletal Mesh=$User.SkeletalMesh
    Write To Particle Position=false
   ) {enabled=false}
   &ShapeLocation(
    Cylinder Height=2
    Cylinder Radius=80
    Hemicircle Internal Cap=false
    Hemicircle Y=true
    MaskSpawnGroup=false
    Offset=<0 -10 -10>
    Offset Coordinate Space=::Local
    Offset Mode=::NewEnumerator0
    Override Seed=false
    Shape Primitive=::NewEnumerator1
    Surface Only=true
    Use Endcaps In Surface Only Mode=false
    UseOwnerScale=true
   )
   &MeshRotationForce(Rotation=<2 2 2>)
   &AddVelocity(
    Rotation Coordinate Space=::Simulation
    Velocity=&RandomRangeVector(Minimum=<0 300 -100>, Maximum=<1 800 100>)
   ) {module=AddVelocity001}
   &ApplyInitialForces()
   &InitialMeshOrientation(Use Rotation Vector=false)
   /SetVariables(Particles.VisibilityTag=&RandomRangeInt(Minimum=0, Maximum=2)) {isSetParameters=true}
   &Color(Color=&SelectLinearColorFromArray(
     Array Sampling Mode=::NewEnumerator1
     Direct Array Index=$Particles.VisibilityTag
     Color Selection Array=DataInterface() {propertyValues {ColorData [<rgba 0.1367 1 0.4018 1>, <rgba 1 0.7238 0 1>, <rgba 1 0 0.1432 1>]}}
     Random Seed=?
    ))
  ]
  particleUpdateScript [
   &ParticleState()
   &ScaleMeshSize(Scale Factor=&Multiply_VectorByFloat(
     Vector=<1 1 1>
     Float=&FloatFromCurve(FloatCurve=DataInterface() {propertyValues {Curve {keys [
          {interpMode=RCIM_Cubic, tangentMode=RCTM_User, time=0, value=0}
          <at 0.2 1>
          {interpMode=RCIM_Cubic, tangentMode=RCTM_User, time=0.8, value=1}
          {tangentMode=RCTM_User, time=1, value=0}
         ]}}})
    ))
   &InheritSourceMovement(Applied Position Scale=<0.9 0.9 0.9>, ApplyForce=false, ApplyPosition=true, ApplyPositionFalloff=false) {enabled=false}
   &SolveForcesAndVelocity(Clamp Velocity=false, Limit Acceleration=false)
   &ScaleMeshSize(Scale Factor=&VectorFromFloat(Value=&FloatFromCurve(FloatCurve=DataInterface() {propertyValues {Curve {keys [
          {arriveTangent=13.3321, interpMode=RCIM_Cubic, leaveTangent=13.3321, tangentMode=RCTM_User, time=0, value=0.5}
          {arriveTangent=0.01, interpMode=RCIM_Cubic, leaveTangent=0.01, tangentMode=RCTM_User, time=0.1, value=1}
          {arriveTangent=-0.0917, interpMode=RCIM_Cubic, leaveTangent=-0.0917, tangentMode=RCTM_User, time=0.4, value=1}
          <at 1 0 RCIM_Cubic>
         ]}}}))) {module=ScaleMeshSize001}
   /SetVariables(Particles.StarScale:Vector3f=&Multiply_VectorByFloat(Vector=$Particles.Scale, Float=0.1)) {isSetParameters=true}
  ]
  rendererClasses [&NiagaraMeshRendererProperties]
  renderers [
   {rendererClass=&NiagaraMeshRendererProperties, rendererIndex=0}
   {rendererClass=&NiagaraMeshRendererProperties, rendererIndex=1}
   {rendererClass=&NiagaraMeshRendererProperties, rendererIndex=2}
   {rendererClass=&NiagaraMeshRendererProperties, rendererIndex=3}
   {rendererClass=&NiagaraMeshRendererProperties, rendererIndex=4}
   {rendererClass=&NiagaraMeshRendererProperties, rendererIndex=5}
  ]
 }
 Trail_Goopy {
  emitterUpdateScript [
   &EmitterState()
   &SpawnBurst_Instantaneous(Use Spawn Probability=false, UseLoopCountLimit=true)
   &InterpolateOverTime(Rate Of Change=5, Target Float Value:NiagaraFloat=$User.ForwardDot) {enabled=false}
  ]
  particleSpawnScript [
   &InitializeParticle(
    Color=<rgba 0.2969 0.0417 0.1309 1>
    Lifetime=0.4
    Mesh Scale=&Multiply_VectorByFloat(Vector=<1.2 1.35 1.5>, Float=1)
    Mesh Scale Mode=::NewEnumerator4
    Position Offset=<100 100 0>
    UsePositionOffset=false
    Write Material Random=true
   )
   &InitialMeshOrientation(
    Mesh Axis To Orient=<0 0 0>
    Mesh Orientation Mode=::NewEnumerator2
    Orientation Vector=<0 0 1>
    Rotation=<0 0 0.25>
    Use Rotation Vector=true
   ) {module=InitialMeshOrientation001}
   /SetVariables(Particles.OutlineColor:LinearColor=<rgba 0.0583 0.0178 0.2396 1>) {isSetParameters=true}
  ]
  particleUpdateScript [
   &ParticleState(Kill Particles When Lifetime Has Elapsed=false, Let Infinitely Lived Particles Die When Emitter Deactivates=true)
   &SkeletalMeshLocation(
    Bone / Socket Index=3
    Bone Sampling Mode=::NewEnumerator4
    MaskSpawnGroup=false
    Orientation Sampling=::NewEnumerator1
    Sampled Position Offset=<8 0 -20>
    Skeletal Mesh=$User.SkeletalMesh
   )
   &UpdateMeshOrientation(
    Facing Direction=&RotateVectorByQuaternion(Vector To Rotate=<0 1 0>, Quaternion=$Particles.SkeletalMeshLocation.SampledRotation)
    Facing Interpolation=false
    Orientation Method=::NewEnumerator1
    Override Quat=false
    TransformFacing=false
    TransformUp=false
    Up Direction:Vector3f=<1 0 0>
   )
   &ScaleMeshSize(Scale Factor=&MakeVector(
     X=&FloatFromCurve(FloatCurve=DataInterface() {propertyValues {Curve {keys [<at 0 0 RCIM_Cubic>,<at 1 1 RCIM_Cubic>]}}})
     Y=&FloatFromCurve(FloatCurve=DataInterface() {propertyValues {Curve {keys [<at 0.3 0.2 RCIM_Cubic>,<at 1 1 RCIM_Cubic>]}}})
     Z=1
    )) {module=ScaleMeshSize001}
   &ScaleMeshSize(Scale Factor=&MakeVector(
     X=&FloatFromCurve(
      FloatCurve=DataInterface() {propertyValues {Curve {keys [<at 0 0.9 RCIM_Cubic>,<at 1 1 RCIM_Cubic>]}}}
      CurveIndex=$System.Timeline001.NormalizedTime
     )
     Y=&FloatFromCurve(
      FloatCurve=DataInterface() {propertyValues {Curve {keys [<at 0 0.8 RCIM_Cubic>,<at 1 1 RCIM_Cubic>]}}}
      CurveIndex=$System.Timeline001.NormalizedTime
     )
     Z=1
    ))
   &ScaleColor(
    Scale Alpha=&FloatFromCurve(
     FloatCurve=DataInterface() {propertyValues {Curve {keys [<at 0 1 RCIM_Cubic>,<at 1 0 RCIM_Cubic>]}}}
     CurveIndex=$System.Timeline001.NormalizedTime
    )
    ScaleA=true
    ScaleRGB=false
   )
  ]
  rendererClasses [&NiagaraMeshRendererProperties]
  renderers [
   {rendererClass=&NiagaraMeshRendererProperties, rendererIndex=0}
   {rendererClass=&NiagaraMeshRendererProperties, rendererIndex=1}
   {rendererClass=&NiagaraMeshRendererProperties, rendererIndex=2}
   {rendererClass=&NiagaraMeshRendererProperties, rendererIndex=3}
  ]
 }
}
summary {
 systemName=NS_Glider_Dazzle_Trails
 emitters [
  {emitterName=Rainbow_Trails, rendererClasses [&NiagaraMeshRendererProperties]}
  {emitterName=StarToon, rendererClasses [&NiagaraMeshRendererProperties]}
  {emitterName=Trail_Goopy, rendererClasses [&NiagaraMeshRendererProperties]}
  {emitterName=SpeedLines, rendererClasses [&NiagaraSpriteRendererProperties]}
 ]
 userVariables [
  User.bIsFrontEnd {type=NiagaraBool}
  User.bIsFrontEndPreview {type=NiagaraBool}
  User.bIsFullyDeployed {type=NiagaraBool, defaultValue=true}
  User.ForwardDot {type=NiagaraFloat, defaultValue=1, description=""}
  User.PlayerSpeed {type=NiagaraFloat, defaultValue=0, description=""}
  User.RibbonIDOffset {type=NiagaraInt32}
  User.RightDot {type=NiagaraFloat, defaultValue=0, description=""}
  User.SkeletalMesh {
   type=NiagaraDataInterfaceSkeletalMesh
   defaultValue {
    struct=NiagaraExt_VariableValue_DataInterface
    value {dataInterface=&self:NiagaraDataInterfaceSkeletalMesh_0}
   }
  }
 ]
}
